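; Parent Blueprint for quest hint props that are shown or hidden based on the local player's quest state.
; On BeginPlay, after a fixed delay marked as a workaround, only non-authority instances check the first local player controller against the quest item and objective name variables.
; The actor hides when the quest is invalid and the objective is not completed.
; The construction script disables collision, and the overlap and tick events are empty.
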(event EventBeginPlay
  ;[1 0.3415 0] Work Around until better way found
  (Utilities|FlowControl|Delay 5.0)
  ;[1 1 1] Check Each Controller to see if they have the quest for this Prop AND haven't already completed the quest - make invisible if not
  (Networking|SwitchHasAuthority
    (:Remote
      (bind (_questisvalid _questcompleted _objectivecompleted) (Default|AthenaCheckQuestandObjectiveCompletion (Utilities|Array|Get(acopy) (Fortnite|GetLocalFortPlayerControllers true)) (Variables|QuestStuff|GetQuestItem) (Variables|QuestStuff|GetObjectiveBackendName)))
      (if (not (or _questisvalid _objectivecompleted))
        (Rendering|SetActorHiddenInGame true)))))

(event Collision|EventActorBeginOverlap
  (params (OtherActor :type Actor%20Object%20Reference)))

(event EventTick
  (params (DeltaSeconds :type Float%20%28single-precision%29)))

(fn ConstructionScript
  (Collision|SetActorEnableCollision false))
